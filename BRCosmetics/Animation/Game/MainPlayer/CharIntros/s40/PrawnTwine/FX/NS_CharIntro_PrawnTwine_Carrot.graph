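# Niagara system for the PrawnTwine character intro that throws orange carrot debris mesh particles from a sampled skeletal mesh location.
# Three size-graded spawn groups come from instant bursts plus a randomized spawn rate scaled by User.Chewing.
# Pieces fly along the sampled orientation's axes with random side flips, then fall under gravity and drag, shrinking late in life with a front-end scale modifier.
# BurstIndex-triggered spawning is present but disabled.
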NiagaraSystem /BRCosmetics/Animation/Game/MainPlayer/CharIntros/s40/PrawnTwine/FX/NS_CharIntro_PrawnTwine_Carrot
dependencies {
 usedDataInterfaces [&NiagaraDataInterfaceCurve, &NiagaraDataInterfaceColorCurve, &NiagaraDataInterfaceSkeletalMesh]
 usedDynamicInputs [&MaskVectorBySpawnGroup]
 usedModules [
  &SystemState
  &self:SystemSpawnScript.SystemScriptSource.SystemScriptGraph.NiagaraNodeAssignment_0.SetVariables_#1
 ]
}
emitters {CarrotDebris {
  emitterUpdateScript [
   &EmitterState(Scalability Mode=::NewEnumerator1)
   &SpawnRate(
    Spawn Probability=&RandomRangeFloat(Minimum=0.4, Maximum=1, Evaluation Type=::NewEnumerator1)
    SpawnGroup:NiagaraInt32=&MakeIntFromFloat(Float=&FloatFromCurve(
      FloatCurve=DataInterface() {propertyValues {Curve {keys [<at 0 0 RCIM_Constant>,<at 0.2521 1 RCIM_Constant>,<at 0.6 2 RCIM_Constant>]}}}
      CurveIndex=&RandomRangeFloat(Minimum=0, Maximum=1)
     ))
    SpawnRate=&Lerp_Float(A=0, B=25, Alpha=$User.Chewing)
    Use Spawn Probability=true
   )
   &SpawnBurst_Instantaneous(
    Spawn Count=&RandomRangeInt(Minimum=1, Maximum=2)
    Use Spawn Probability=false
    UseLoopCountLimit=false
   )
   &SpawnBurst_Instantaneous(
    Spawn Count=&RandomRangeInt(Minimum=4, Maximum=7)
    Spawn Group=1
    Use Spawn Probability=false
    UseLoopCountLimit=false
   ) {module=SpawnBurst_Instantaneous001}
   &SpawnBurst_Instantaneous(
    Spawn Count=&RandomRangeInt(Minimum=6, Maximum=12)
    Spawn Group=2
    Use Spawn Probability=false
    UseLoopCountLimit=false
   ) {module=SpawnBurst_Instantaneous002}
   /Game/Effects/Niagara/Modules/Logic/UpdateValueStateChange(InputIntParam:NiagaraInt32=&MakeIntFromFloat(Float=$User.BurstIndex, Conversion Mode=::NewEnumerator1)) {enabled=false}
   &self:TriggerSpawn(
    Spawn Count:NiagaraInt32=&RandomRangeInt(Minimum=3, Maximum=6)
    Trigger:NiagaraBool=&Boolean_AndOperation(Bool1=$Emitter.UpdateValueStateChange.Changed, Bool2=&SetBoolByFloatComparison(A=$System.Age, B=0))
   ) {enabled=false, module=TriggerSpawn}
   &self:TriggerSpawn(
    Spawn Count:NiagaraInt32=&RandomRangeInt(Minimum=20, Maximum=30)
    Spawn Group=1
    Trigger:NiagaraBool=&Boolean_AndOperation(Bool1=$Emitter.UpdateValueStateChange.Changed, Bool2=&SetBoolByFloatComparison(A=$System.Age, B=0))
   ) {enabled=false, module=TriggerSpawn001}
  ]
  particleSpawnScript [
   &InitializeParticle(
    Color=<rgba 1 0.1021 0 1>
    Lifetime=&RandomRangeFloat(Minimum=0.3, Maximum=0.5)
    Mass Max=1
    Mass Min=0.9
    Mass Mode=::NewEnumerator3
    Mesh Scale=&MergeHelper_VectorSwitch(
     0=&RandomRangeVector(Minimum=<0.6 0.1 0.6>, Maximum=<1.05 0.3 1.05>)
     1=&RandomRangeVector(Minimum=<0.42 0.07 0.42>, Maximum=<0.735 0.21 0.735>)
     2=&RandomRangeVector(Minimum=<0.294 0.049 0.294>, Maximum=<0.5145 0.147 0.5145>)
     3=?
     Index=$Emitter.SpawnGroup
     Num=::NewEnumerator1
    )
    Mesh Scale Mode=::NewEnumerator4
    UsePositionOffset=false
    Write Material Random=true
   )
   /SetVariables() {isSetParameters=true}
   &SkeletalMeshLocation(
    MaskSpawnGroup=false
    Mesh Sampling Type=::NewEnumerator5
    Orientation Sampling=::NewEnumerator1
    Sampled Position Offset=<6 10 -2>
    Skeletal Mesh=$System.Skeletal Mesh
    Socket Sampling Mode=::NewEnumerator5
   )
   &ShapeLocation(MaskSpawnGroup=false, Override Seed=false, Shape Origin=$Particles.SkeletalMeshLocation.SampledPosition, Sphere Radius=3, UseOwnerScale=true) {enabled=false}
   &InitialMeshOrientation(Rotation=&RandomRangeVector(Minimum=<-1 -1 0>, Maximum=<1 1 1>), Use Rotation Vector=true)
   &AddVelocity(
    Coordinate Space=::World
    Velocity=&MakeCustomVectorFromBool(
     Bool=$Particles.RandomSide
     True Vector=&Multiply_VectorByFloat(
      Vector=&ReturnMeshOrientationAxis(Mesh Orientation Axis=::ZAxis, Mesh Orientation Quat=$Particles.SkeletalMeshLocation.SampledRotation)
      Float=&RandomRangeFloat(Minimum=-10, Maximum=-80)
     )
     False Vector=&Multiply_VectorByFloat(
      Vector=&ReturnMeshOrientationAxis(Mesh Orientation Axis=::ZAxis, Mesh Orientation Quat=$Particles.SkeletalMeshLocation.SampledRotation)
      Float=&RandomRangeFloat(Minimum=20, Maximum=70)
     )
    )
   )
   &AddVelocity(
    Coordinate Space=::World
    Velocity=&Multiply_VectorByFloat(
     Vector=&ReturnMeshOrientationAxis(Mesh Orientation Axis=::XAxis, Mesh Orientation Quat=$Particles.SkeletalMeshLocation.SampledRotation)
     Float=&MaskFloatBySpawnGroup(
      Masked Float=&RandomRangeFloat(Minimum=25, Maximum=50)
      Passthrough Float=&MaskFloatBySpawnGroup(
       Masked Float=&RandomRangeFloat(Minimum=25, Maximum=60)
       Passthrough Float=&RandomRangeFloat(Minimum=70, Maximum=90)
       Spawn Group Mask=1
      )
     )
    )
   ) {module=AddVelocity005}
   &AddVelocity(
    Coordinate Space=::World
    Velocity=&Multiply_VectorByFloat(
     Vector=&ReturnMeshOrientationAxis(Mesh Orientation Axis=::YAxis, Mesh Orientation Quat=$Particles.SkeletalMeshLocation.SampledRotation)
     Float=&MaskFloatBySpawnGroup(
      Masked Float=&RandomRangeFloat(Minimum=50, Maximum=70)
      Passthrough Float=&MaskFloatBySpawnGroup(
       Masked Float=&RandomRangeFloat(Minimum=50, Maximum=60)
       Passthrough Float=&RandomRangeFloat(Minimum=100, Maximum=130)
       Spawn Group Mask=1
      )
     )
    )
   ) {module=AddVelocity004}
   /SetVariables(Particles.Velocity=&Multiply_VectorByFloat(
     Vector=$Particles.Velocity
     Float=&MergeHelper_FloatSwitch(
      0=&RandomRangeFloat(Minimum=1, Maximum=0.6)
      1=&RandomRangeFloat(Minimum=0.8, Maximum=0.4)
      2=&RandomRangeFloat(Minimum=0.2, Maximum=0.4)
      3=?
      Index=$Emitter.SpawnGroup
      Num=::NewEnumerator1
     )
    )) {isSetParameters=true}
  ]
  particleUpdateScript [
   &GravityForce(Gravity=&RandomRangeVector(Minimum=<0 0 -400>, Maximum=<0 0 -800>))
   &ScaleColor(
    Scale RGB=&MakeVectorFromLinearColorRGB(Color=&ColorFromCurve(
      ColorCurve=DataInterface() {propertyValues {
        AlphaCurve {keys [<at 0 1 RCIM_Cubic>]}
        BlueCurve {keys [
          <at 0.2162 0.7 RCIM_Cubic>
          <at 0.2812 0.6314>
          {arriveTangent=1.3318, interpMode=RCIM_Cubic, leaveTangent=1.3318, time=0.3527, value=1}
          {arriveTangent=-1.0217, interpMode=RCIM_Cubic, leaveTangent=-1.0217, time=0.558, value=1}
          {arriveTangent=-0.8311, interpMode=RCIM_Cubic, leaveTangent=-0.8311, time=0.8029, value=0.54}
          <at 0.919 0.7>
         ]}
        GreenCurve {keys [
          <at 0.2162 0.35 RCIM_Cubic>
          <at 0.2812 0.6667>
          {arriveTangent=1.2043, interpMode=RCIM_Cubic, leaveTangent=1.2043, time=0.3527, value=1}
          {arriveTangent=-0.8536, interpMode=RCIM_Cubic, leaveTangent=-0.8536, time=0.558, value=1}
          {arriveTangent=-1.8007, interpMode=RCIM_Cubic, leaveTangent=-1.8007, time=0.8029, value=0.6157}
          <at 0.919 0.35>
         ]}
        RedCurve {keys [
          <at 0.2162 0.28 RCIM_Cubic>
          <at 0.2812 0.902>
          {arriveTangent=0.3542, interpMode=RCIM_Cubic, leaveTangent=0.3542, time=0.3527, value=1}
          {arriveTangent=-0.2221, interpMode=RCIM_Cubic, leaveTangent=-0.2221, time=0.558, value=1}
          {arriveTangent=-1.9947, interpMode=RCIM_Cubic, leaveTangent=-1.9947, time=0.8029, value=0.9}
          <at 0.919 0.28>
         ]}
       }}
      CurveIndex=Linked() {linkedVariable {name=NPC.FortniteNPC.RemappedTimeOfDay_0-1, type=NiagaraFloat}}
     ))
    ScaleA=false
    ScaleRGB=false
   ) {module=ScaleColor001}
   &Drag(Drag=&RandomRangeFloat(Minimum=0.9, Maximum=1.1), Use Linear Drag=true, Use Rotational Drag=false)
   &MeshRotationForce(Rotation=&RandomRangeVector(Minimum=<25 0 0>, Maximum=<30 10 0>))
   &InheritSourceMovement(
    Applied Position Falloff Distance=1000
    Applied Position Scale=&Multiply_VectorByFloat(
     Vector=<1 1 0>
     Float=&FloatFromCurve(
      FloatCurve=DataInterface() {propertyValues {Curve {keys [
          {interpMode=RCIM_Cubic, tangentMode=RCTM_User, time=0.1865, value=1}
          {arriveTangent=-2.1905, interpMode=RCIM_Cubic, leaveTangent=-2.1905, tangentMode=RCTM_User, time=1, value=0}
         ]}}}
      CurveIndex=$Emitter.Age
     )
    )
    ApplyForce=false
    ApplyPosition=true
    ApplyPositionFalloff=false
   )
   &SolveForcesAndVelocity(Clamp Velocity=false, Limit Acceleration=false)
   &ParticleState()
   &ScaleMeshSize(Scale Factor=&VectorFromFloat(Value=&FloatFromCurve(
      FloatCurve=DataInterface() {propertyValues {Curve {keys [
          {tangentMode=RCTM_User, time=0, value=1}
          {arriveTangent=-0.0405, interpMode=RCIM_Cubic, leaveTangent=-0.0405, tangentMode=RCTM_User, time=0.5, value=1}
          {arriveTangent=-4.4728, interpMode=RCIM_Cubic, leaveTangent=-4.4728, tangentMode=RCTM_User, time=1, value=0}
         ]}}}
      Scale Curve=&FN_FrontEndFloatModifier(Gameplay=1, Front End=$User.FrontEndScale)
     )))
  ]
  rendererClasses [&NiagaraMeshRendererProperties]
  renderers [{rendererClass=&NiagaraMeshRendererProperties, rendererIndex=0}]
 }}
summary {
 systemName=NS_CharIntro_PrawnTwine_Carrot
 emitters [{emitterName=CarrotDebris, rendererClasses [&NiagaraMeshRendererProperties]}]
 userVariables [
  User.bIsFrontEnd {type=NiagaraBool}
  User.bIsFrontEndPreview {type=NiagaraBool}
  User.BurstIndex {type=NiagaraFloat, defaultValue=0, description=""}
  User.Chewing {type=NiagaraFloat, defaultValue=0, description=""}
  User.FrontEndScale {type=NiagaraFloat, defaultValue=0.8, description=""}
 ]
}
